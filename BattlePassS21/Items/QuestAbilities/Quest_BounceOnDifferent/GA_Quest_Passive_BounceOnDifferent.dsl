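; Passive quest ability that tracks the player bouncing on different items for the BounceOnDifferent quest.
; Activation chains three async listeners on the owning pawn: Slurpshroom-used event, tire launch tag added, and AppleSunUsed event for the Crash Pad.
; Each callback sets its own boolean flag.
; An AND of all three flags gates a complex custom stat event sent to the player's quest manager.
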(event Ability|EventOnEndAbility
  (params (bWasCancelled :type Boolean))
  (Ability|EndAbility))

(event Ability|EventFailedToActivatePassiveAbility
  (params (ActorInfo :type Gameplay%20Ability%20Actor%20Info%20Structure))
  (Ability|EndAbility))

(event Ability|EventActivateAbility
  (|CallParentFunction)
  (Ability|Async|WaitGameplayEventToActor (Variables|GAVars|Data|GetOwningPlayerPawn) "(TagName=\"Athena.Quests.BouncySlurpshroomUsed\")" true
    (:then
      (Ability|Tasks|WaitGameplayTagAddToActor (Variables|GAVars|Data|GetOwningPlayerPawn) "(TagName=\"Granted.Athena.EnvItem.Launcher.Tire.Launch\")" true
        (:then
          (Ability|Async|WaitGameplayEventToActor (Variables|GAVars|Data|GetOwningPlayerPawn) "(TagName=\"Event.AppleSunUsed\")" true
            (:EventReceived
              (Variables|Default|SetCrashPad true))))
        (:Added
          (Variables|Default|SetTires true))))
    (:EventReceived
      (Variables|Default|SetSlurpShroom true)))
  (if (Math|Boolean|ANDBoolean (Variables|Default|GetSlurpShroom) (Variables|Default|GetTires) (Variables|Default|GetCrashPad))
    (Quest|SendComplexCustomStatEvent (Variables|GAVars|Data|GetOwningPlayerQuestManager) 0 "(GameplayTags=((TagName=\"Athena.Quests.BounceOnDifferent\")))" "(GameplayTags=)" 1)))
